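# Physics asset for the Dino Hunter backpack cosmetic, providing capsule bodies and joints for its dangling chain, skull and feathers.
# The pack body anchors a two-link bag chain and the skull, which in turn carries base and front feathers, some as two-segment chains.
# Every constraint locks twist, and the skull and most feather joints also lock the second swing axis, so they swing on a single axis.
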PhysicsAsset /BRCosmetics/Accessories/FORT_Backpacks/Backpack_F_MED_Dino_Hunter/Meshes/F_MED_Dino_Hunter_Pack_Physics
bodies {
 dyn_bagChain_1 [{shapeType=Capsule, length=11, radius=2, center=<1.3214 0 -0.371>, rotation=<rot -15.6842 0 0>}]
 dyn_bagChain_2 [{shapeType=Capsule, length=11, radius=2, center=<1.3725 0 0>, rotation=<rot 0 0 7.8523>}]
 dyn_baseFeather_jnt_1 [{shapeType=Capsule, length=2, radius=0.6, center=<0.442 0 -0.0054>, rotation=<rot 89.3048 -0.0002 -0.0002>}]
 dyn_baseFeather_jnt_2 [{shapeType=Capsule, length=2, radius=0.6, center=<0.442 0 -0.0055>, rotation=<rot 89.2964 -0.0012 -0.0012>}]
 dyn_baseFeather_jnt_2_2 [{shapeType=Capsule, length=2, radius=0.6, center=<0.442 0 -0.0055>, rotation=<rot 89.2964 -0.0013 -0.0014>}]
 dyn_baseFeather_jnt_3 [{shapeType=Capsule, length=2, radius=1, center=<0.3272 0.1296 -0.0041>, rotation=<rot 89.2939 -0.0012 -21.6059>}]
 dyn_baseFeather_jnt_3_2 [{shapeType=Capsule, length=2.2477, radius=1.5, center=<1.17 0 -0.0142>, rotation=<rot 89.3068 0.0076 0.0076>}]
 dyn_baseFeather_jnt_4 [{shapeType=Capsule, length=2, radius=1, center=<0.6572 0 -0.0081>, rotation=<rot 89.2964 -0.0012 -0.0012>}]
 dyn_baseFeather_jnt_5 [{shapeType=Capsule, length=2, radius=1, center=<0.442 0 -0.0055>, rotation=<rot 89.2964 0.0026 0.0026>}]
 dyn_baseFeather_jnt_5_2 [{shapeType=Capsule, length=2, radius=1, center=<0.5949 0 -0.0073>, rotation=<rot 89.2989 0.0018 0.0017>}]
 dyn_baseFeather_jnt_6 [{shapeType=Capsule, length=2, radius=0.6, center=<0.442 0 -0.0053>, rotation=<rot 89.3164 0.0015 0.0013>}]
 dyn_baseFeather_jnt_7 [{shapeType=Capsule, length=2, radius=0.6, center=<0.233 0 -0.0029>, rotation=<rot 89.3003 -0.0024 -0.0024>}]
 dyn_frontFeather_jnt_1 [{shapeType=Capsule, length=6, radius=0.6, center=<1.2273 0 -0.0149>, rotation=<rot 89.309 0.0076 0.0076>}]
 dyn_frontFeather_jnt_2 [{shapeType=Capsule, length=4, radius=0.6, center=<0.8433 0 -0.0103>, rotation=<rot 89.3017 -0.0008 -0.0008>}]
 dyn_frontFeather_jnt_3 [{shapeType=Capsule, length=4, radius=0.6, center=<0.442 0 -0.0055>, rotation=<rot 89.2989 0.0019 0.0019>}]
 dyn_skull_jnt [{shapeType=Capsule, length=10, radius=5, center=<1.5544 -2.1309 -0.1483>, rotation=<rot -89.1851 -180 180>}]
 pack [{shapeType=Capsule, length=30, radius=5, center=<-1.32 0.0003 -4.7664>, rotation=<rot -34.6966 0 0>}]
}
constraints [
 {bone1Name=dyn_bagChain_1, bone2Name=pack, swing1LimitDegrees=20, swing2LimitDegrees=10, twistLimitDegrees=20, twistMotion=Locked}
 {bone1Name=dyn_bagChain_2, bone2Name=dyn_bagChain_1, swing1LimitDegrees=30, swing2LimitDegrees=40, twistLimitDegrees=20, twistMotion=Locked}
 {
  bone1Name=dyn_skull_jnt
  bone2Name=pack
  swing1LimitDegrees=20
  swing2LimitDegrees=45
  swing2Motion=Locked
  twistLimitDegrees=45
  twistMotion=Locked
 }
 {
  bone1Name=dyn_baseFeather_jnt_1
  bone2Name=dyn_skull_jnt
  swing1LimitDegrees=5
  swing2LimitDegrees=5
  swing2Motion=Locked
  twistLimitDegrees=5
  twistMotion=Locked
 }
 {
  bone1Name=dyn_baseFeather_jnt_2
  bone2Name=dyn_skull_jnt
  swing1LimitDegrees=3
  swing2LimitDegrees=3
  twistLimitDegrees=5
  twistMotion=Locked
 }
 {
  bone1Name=dyn_baseFeather_jnt_2_2
  bone2Name=dyn_baseFeather_jnt_2
  swing1LimitDegrees=5
  swing2LimitDegrees=5
  swing2Motion=Locked
  twistLimitDegrees=5
  twistMotion=Locked
 }
 {
  bone1Name=dyn_baseFeather_jnt_3
  bone2Name=dyn_skull_jnt
  swing1LimitDegrees=5
  swing2LimitDegrees=5
  swing2Motion=Locked
  twistLimitDegrees=5
  twistMotion=Locked
 }
 {
  bone1Name=dyn_baseFeather_jnt_3_2
  bone2Name=dyn_baseFeather_jnt_3
  swing1LimitDegrees=5
  swing2LimitDegrees=5
  swing2Motion=Locked
  twistLimitDegrees=5
  twistMotion=Locked
 }
 {
  bone1Name=dyn_baseFeather_jnt_4
  bone2Name=dyn_skull_jnt
  swing1LimitDegrees=5
  swing2LimitDegrees=5
  twistLimitDegrees=5
  twistMotion=Locked
 }
 {
  bone1Name=dyn_baseFeather_jnt_5
  bone2Name=dyn_skull_jnt
  swing1LimitDegrees=5
  swing2LimitDegrees=5
  twistLimitDegrees=5
  twistMotion=Locked
 }
 {
  bone1Name=dyn_baseFeather_jnt_5_2
  bone2Name=dyn_baseFeather_jnt_5
  swing1LimitDegrees=5
  swing2LimitDegrees=5
  swing2Motion=Locked
  twistLimitDegrees=5
  twistMotion=Locked
 }
 {
  bone1Name=dyn_baseFeather_jnt_6
  bone2Name=dyn_skull_jnt
  swing1LimitDegrees=5
  swing2LimitDegrees=5
  swing2Motion=Locked
  twistLimitDegrees=5
  twistMotion=Locked
 }
 {
  bone1Name=dyn_baseFeather_jnt_7
  bone2Name=dyn_skull_jnt
  swing1LimitDegrees=5
  swing2LimitDegrees=5
  twistLimitDegrees=5
  twistMotion=Locked
 }
 {
  bone1Name=dyn_frontFeather_jnt_1
  bone2Name=dyn_skull_jnt
  swing1LimitDegrees=3
  swing2LimitDegrees=5
  swing2Motion=Locked
  twistLimitDegrees=5
  twistMotion=Locked
 }
 {
  bone1Name=dyn_frontFeather_jnt_2
  bone2Name=dyn_skull_jnt
  swing1LimitDegrees=3
  swing2LimitDegrees=5
  swing2Motion=Locked
  twistLimitDegrees=5
  twistMotion=Locked
 }
 {
  bone1Name=dyn_frontFeather_jnt_3
  bone2Name=dyn_skull_jnt
  swing1LimitDegrees=3
  swing2LimitDegrees=5
  swing2Motion=Locked
  twistLimitDegrees=5
  twistMotion=Locked
 }
]
