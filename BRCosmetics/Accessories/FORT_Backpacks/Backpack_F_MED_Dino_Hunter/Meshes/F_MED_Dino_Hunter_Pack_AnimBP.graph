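# Animation Blueprint for the Dino Hunter backpack cosmetic, animating its feathers, chain and pack placement.
# RigidBody physics varies blend weight, gravity and motion scaling by front-end, skydiving and ground state.
# Perlin-style random rotators, updated each frame, additively sway feather bones atop the physics, alongside fixed corrective offsets.
# Skydiving adds chain and pack offsets; dive mode blends to the cached pose with the pack transform replaced.
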AnimBlueprint /BRCosmetics/Accessories/FORT_Backpacks/Backpack_F_MED_Dino_Hunter/Meshes/F_MED_Dino_Hunter_Pack_AnimBP
target_skeleton=&/F_MED_Dino_Hunter_Pack_Skeleton
graphs [
 AnimGraph {nodes [
   n1 = "Transform (Modify) Bone - Bone: pack"(ComponentPose=Local To Component(LocalPose=Use cached pose 'Not Dive Mode'()), Translation="0,-5,0", Rotation="-7,0,0") {rotationMode=BMM_Replace, alphaBoolBlend {blendOption=Linear}, boneToModify {boneName=pack}, rotation {pitch=-7}, translation {y=-5}}
   Output Pose(Result=Blend Poses by bool(BlendPose_0=Component To Local(ComponentPose=@n1), BlendPose_1=Use cached pose 'Not Dive Mode'(), bActiveValue=Get bIsSkydiveDiveMode()) {blendType=Linear}) {layerGroup=None}
   n2 = Get bIsOnGround() -> (bIsOnGround:bool)
   n3 = Get bIsInFrontEnd()
   n4 = Get bIsSkyDiving()
   n5 = RigidBody!(
    ComponentPose=Local To Component(LocalPose=Input Pose())
    Alpha=SelectFloat(A=0.3, B=1, bPickA=OR Boolean(A=@n3, B=@n4))
    OverrideWorldGravity:Vector=SelectVector(A="0,0,-3000", B="0,0,-600", bPickA=Get bIsInFrontEnd())
    ExternalForce
    ComponentLinearAccScale:Vector=SelectVector(A="1.2,0.35,0.25", B="0.5,0.5,0.25", bPickA=@n2)
    ComponentLinearVelScale:Vector=SelectVector(A="0.8,0.8,0.1", B="0.1,0.1,0.1", bPickA=@n2)
   ) -> (Pose) {
    bClampLinearTranslationLimitToRefPose=true
    bOverrideWorldGravity=true
    overridePhysicsAsset=&/F_MED_Dino_Hunter_Pack_Physics
    alphaBoolBlend {blendOption=Linear}
    baseBoneRef {boneName=root}
    componentAppliedLinearAccClamp=<6000 6000 6000>
    componentLinearAccScale=<0.25 0.25 0.25>
    componentLinearVelScale=<0.1 0.1 0.1>
    overrideWorldGravity {z=-600}
   }
   n6 = Get FeatherRotator1() -> (FeatherRotator1:Rotator)
   n7 = "Transform (Modify) Bone - Bone: dyn_baseFeather_jnt_1"(ComponentPose=@n5, Alpha=1, Translation="0,0,0", Rotation=@n6, Scale="1,1,1") {rotationMode=BMM_Additive, alphaBoolBlend {blendOption=Linear}, boneToModify {boneName=dyn_baseFeather_jnt_1}}
   n8 = Get FeatherRotator2() -> (FeatherRotator2:Rotator)
   n9 = Get FeatherRotator3() -> (FeatherRotator3:Rotator)
   n10 = "Transform (Modify) Bone - Bone: dyn_baseFeather_jnt_3_2"(
    ComponentPose="Transform (Modify) Bone - Bone: dyn_baseFeather_jnt_2_2"(
     ComponentPose="Transform (Modify) Bone - Bone: dyn_baseFeather_jnt_2"(ComponentPose=@n7, Alpha=1, Translation="0,0,0", Rotation=@n8, Scale="1,1,1") {rotationMode=BMM_Additive, alphaBoolBlend {blendOption=Linear}, boneToModify {boneName=dyn_baseFeather_jnt_2}}
     Alpha=1
     Translation="0,0,0"
     Rotation=@n9
     Scale="1,1,1"
    ) {rotationMode=BMM_Additive, alphaBoolBlend {blendOption=Linear}, boneToModify {boneName=dyn_baseFeather_jnt_2_2}}
    Alpha=1
    Translation="0,0,0"
    Rotation=@n8
    Scale="1,1,1"
   ) {rotationMode=BMM_Additive, alphaBoolBlend {blendOption=Linear}, boneToModify {boneName=dyn_baseFeather_jnt_3_2}}
   n11 = "Transform (Modify) Bone - Bone: dyn_baseFeather_jnt_5_2"(
    ComponentPose="Transform (Modify) Bone - Bone: dyn_baseFeather_jnt_5"(
     ComponentPose="Transform (Modify) Bone - Bone: dyn_baseFeather_jnt_4"(ComponentPose=@n10, Alpha=1, Translation="0,0,0", Rotation=@n8, Scale="1,1,1") {rotationMode=BMM_Additive, alphaBoolBlend {blendOption=Linear}, boneToModify {boneName=dyn_baseFeather_jnt_4}}
     Alpha=1
     Translation="0,0,0"
     Rotation=@n9
     Scale="1,1,1"
    ) {rotationMode=BMM_Additive, alphaBoolBlend {blendOption=Linear}, boneToModify {boneName=dyn_baseFeather_jnt_5}}
    Alpha=1
    Translation="0,0,0"
    Rotation=@n9
    Scale="1,1,1"
   ) {rotationMode=BMM_Additive, alphaBoolBlend {blendOption=Linear}, boneToModify {boneName=dyn_baseFeather_jnt_5_2}}
   n12 = "Transform (Modify) Bone - Bone: dyn_frontFeather_jnt_1"(
    ComponentPose="Transform (Modify) Bone - Bone: dyn_frontFeather_jnt_1"(
     ComponentPose="Transform (Modify) Bone - Bone: dyn_baseFeather_jnt_7"(ComponentPose=@n11, Alpha=1, Translation="0,0,0", Rotation=@n6, Scale="1,1,1") {rotationMode=BMM_Additive, alphaBoolBlend {blendOption=Linear}, boneToModify {boneName=dyn_baseFeather_jnt_7}}
     Alpha=1
     Translation="0,0,0"
     Rotation=@n8
     Scale="1,1,1"
    ) {rotationMode=BMM_Additive, alphaBoolBlend {blendOption=Linear}, boneToModify {boneName=dyn_frontFeather_jnt_1}}
    Alpha=1
    Translation="0,0,0"
    Rotation="0,0,-6.5"
    Scale="1,1,1"
   ) {rotationMode=BMM_Additive, alphaBoolBlend {blendOption=Linear}, boneToModify {boneName=dyn_frontFeather_jnt_1}, rotation {roll=-6.5}}
   n13 = "Transform (Modify) Bone - Bone: dyn_frontFeather_jnt_2"(ComponentPose=@n12, Alpha=1, Translation="0,0,0", Rotation="0,0,-5", Scale="1,1,1") {rotationMode=BMM_Additive, alphaBoolBlend {blendOption=Linear}, boneToModify {boneName=dyn_frontFeather_jnt_2}, rotation {roll=-5}}
   n14 = "Transform (Modify) Bone - Bone: dyn_frontFeather_jnt_3"(ComponentPose=@n13, Alpha=1, Translation="0,0,0", Rotation="0,0,-2", Scale="1,1,1") {rotationMode=BMM_Additive, alphaBoolBlend {blendOption=Linear}, boneToModify {boneName=dyn_frontFeather_jnt_3}, rotation {roll=-2}}
   n15 = "Transform (Modify) Bone - Bone: pack"(ComponentPose=@n14, Translation="1.9,-0.5,0", Rotation="4,0,3") {
    rotationMode=BMM_Additive
    translationMode=BMM_Additive
    alphaBoolBlend {blendOption=Linear}
    boneToModify {boneName=pack}
    rotation {pitch=4, roll=3}
    translation=<1.9 -0.5>
   }
   n16 = "To Float (Boolean)"(InBool=Get bIsSkyDiving())
   n17 = "Transform (Modify) Bone - Bone: dyn_bagChain_1"(ComponentPose=@n15, Alpha=@n16, Translation="0,0,0", Rotation="0,0,20", Scale="1,1,1") {rotationMode=BMM_Additive, alphaBoolBlend {blendOption=Linear}, boneToModify {boneName=dyn_bagChain_1}, rotation {roll=20}}
   n18 = "Transform (Modify) Bone - Bone: dyn_bagChain_2"(ComponentPose=@n17, Alpha=@n16, Translation="0,0,0", Rotation="0,0,35", Scale="1,1,1") {rotationMode=BMM_Additive, alphaBoolBlend {blendOption=Linear}, boneToModify {boneName=dyn_bagChain_2}, rotation {roll=35}}
   n19 = "Transform (Modify) Bone - Bone: pack"(ComponentPose=@n18, Alpha=@n16, Translation="0,-5,0", Rotation="-15,0,0") {
    rotationMode=BMM_Additive
    translationMode=BMM_Additive
    alphaBoolBlend {blendOption=Linear}
    boneToModify {boneName=pack}
    rotation {pitch=-15}
    translation {y=-5}
   }
   Save cached pose 'Not Dive Mode'(Pose=Component To Local(ComponentPose=@n19)) {cacheName=Not Dive Mode}
  ]}
 EventGraph {nodes [
   Event BlueprintUpdateAnimation()
   n1 = GetGameTimeInSeconds()
   > n2 = RandomRotatorFunction!(
    execute
    GameTimeInSeconds:double=@n1
    PerlinScale:double=1
    Frequency:Rotator="0, 0,1.3"
    Amplitude:Rotator="0, 0,6"
    Offset:Rotator="0, 0,1"
    TimeOffset:double=0
   ) -> (then, RandomRotator:Rotator)
   > Set FeatherRotator1(FeatherRotator1:Rotator=@n2) -> (Output_Get:Rotator)
   > n3 = RandomRotatorFunction(GameTimeInSeconds=@n1, PerlinScale=1, Frequency="0,0,1.1", Amplitude="0, 0,3", Offset="0, 0,1", TimeOffset=0)
   > Set FeatherRotator2(FeatherRotator2:Rotator=@n3) -> (Output_Get:Rotator)
   > n4 = RandomRotatorFunction(GameTimeInSeconds=@n1, PerlinScale=1, Frequency="0,0,2.4", Amplitude="0, 0,5.5", Offset="0, 0,4", TimeOffset=0)
   > Set FeatherRotator3(FeatherRotator3:Rotator=@n4) -> (Output_Get:Rotator)
  ]}
]
variables [
 FeatherRotator1 {type=Rotator}
 FeatherRotator2 {type=Rotator}
 FeatherRotator3 {type=Rotator}
 rInterp_Skydive {type=Rotator}
]
